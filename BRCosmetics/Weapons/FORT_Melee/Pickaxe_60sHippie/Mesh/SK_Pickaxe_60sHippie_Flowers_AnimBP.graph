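# Animation Blueprint for the flowers on the 60sHippie pickaxe cosmetic, adding procedural secondary motion to seven flower bones.
# A reference pose runs through a chain of single-bone Anim Dynamics nodes, one each for dyn_flower_01 to 07, then to the output.
# Each uses a heavily damped, angle-limited angular spring at low alpha, with slightly varied tuning per flower.
# The event graph has no logic and no variables exist.
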AnimBlueprint /BRCosmetics/Weapons/FORT_Melee/Pickaxe_60sHippie/Mesh/SK_Pickaxe_60sHippie_Flowers_AnimBP
target_skeleton=&/SK_Pickaxe_60sHippie_Flowers_Skeleton
graphs [
 AnimGraph {nodes [
   n1 = Anim Dynamics - Bone: dyn_flower_01(ComponentPose=Local To Component(LocalPose=Local Space Ref Pose()), Alpha=0.2, ExternalForce="0,0,0") {
    angularDampingOverride=0.995
    angularSpringConstant=20000
    bAngularSpring=true
    bGravityOverrideInSimSpace=true
    bOverrideAngularDamping=true
    numSolverIterationsPostUpdate=4
    numSolverIterationsPreUpdate=16
    alphaBoolBlend {blendOption=Linear}
    boundBone {boneName=dyn_flower_01}
    componentLinearAccScale=<1 1 1>
    componentLinearVelScale=<1 1 1>
    physicsBodyDefinitions [{
      sphereCollisionRadius=0
      boundBone {boneName=dyn_flower_01}
      boxExtents=<20 20 10>
      constraintSetup {angularTargetAxis=AxisY, twistAxis=AxisZ, angularLimitsMax=<25 25 2>, angularLimitsMin=<-25 -25 -2>, angularTarget=<0 1 0>}
      localJointOffset=<0 0 -3>
     }]
   }
   n2 = Anim Dynamics - Bone: dyn_flower_02(ComponentPose=@n1, Alpha=0.22, ExternalForce="0,0,0") {
    angularDampingOverride=0.995
    angularSpringConstant=18000
    bAngularSpring=true
    bGravityOverrideInSimSpace=true
    bOverrideAngularDamping=true
    numSolverIterationsPostUpdate=4
    numSolverIterationsPreUpdate=16
    alphaBoolBlend {blendOption=Linear}
    boundBone {boneName=dyn_flower_02}
    componentLinearAccScale=<1 1 1>
    componentLinearVelScale=<1 1 1>
    physicsBodyDefinitions [{
      sphereCollisionRadius=0
      boundBone {boneName=dyn_flower_02}
      boxExtents=<15 25 12>
      constraintSetup {angularTargetAxis=AxisY, twistAxis=AxisZ, angularLimitsMax=<25 25 2>, angularLimitsMin=<-25 -25 -2>, angularTarget=<0 1 0>}
      localJointOffset=<0 0 -4>
     }]
   }
   n3 = Anim Dynamics - Bone: dyn_flower_03(ComponentPose=@n2, Alpha=0.18, ExternalForce="0,0,0") {
    angularDampingOverride=0.995
    angularSpringConstant=21000
    bAngularSpring=true
    bGravityOverrideInSimSpace=true
    bOverrideAngularDamping=true
    numSolverIterationsPostUpdate=4
    numSolverIterationsPreUpdate=16
    alphaBoolBlend {blendOption=Linear}
    boundBone {boneName=dyn_flower_03}
    componentLinearAccScale=<1 1 1>
    componentLinearVelScale=<1 1 1>
    physicsBodyDefinitions [{
      sphereCollisionRadius=0
      boundBone {boneName=dyn_flower_03}
      boxExtents=<18 18 9>
      constraintSetup {angularTargetAxis=AxisY, twistAxis=AxisZ, angularLimitsMax=<25 25 2>, angularLimitsMin=<-25 -25 -2>, angularTarget=<0 1 0>}
      localJointOffset=<0 0 -2.5>
     }]
   }
   n4 = Anim Dynamics - Bone: dyn_flower_04(ComponentPose=@n3, Alpha=0.21, ExternalForce="0,0,0") {
    angularDampingOverride=0.995
    angularSpringConstant=20000
    bAngularSpring=true
    bGravityOverrideInSimSpace=true
    bOverrideAngularDamping=true
    numSolverIterationsPostUpdate=4
    numSolverIterationsPreUpdate=16
    alphaBoolBlend {blendOption=Linear}
    boundBone {boneName=dyn_flower_04}
    componentLinearAccScale=<1 1 1>
    componentLinearVelScale=<1 1 1>
    physicsBodyDefinitions [{
      sphereCollisionRadius=0
      boundBone {boneName=dyn_flower_04}
      boxExtents=<25 30 16>
      constraintSetup {angularTargetAxis=AxisY, twistAxis=AxisZ, angularLimitsMax=<25 25 2>, angularLimitsMin=<-25 -25 -2>, angularTarget=<0 1 0>}
      localJointOffset=<0 0 -5>
     }]
   }
   n5 = Anim Dynamics - Bone: dyn_flower_05(ComponentPose=@n4, Alpha=0.19, ExternalForce="0,0,0") {
    angularDampingOverride=0.995
    angularSpringConstant=19500
    bAngularSpring=true
    bGravityOverrideInSimSpace=true
    bOverrideAngularDamping=true
    numSolverIterationsPostUpdate=4
    numSolverIterationsPreUpdate=16
    alphaBoolBlend {blendOption=Linear}
    boundBone {boneName=dyn_flower_05}
    componentLinearAccScale=<1 1 1>
    componentLinearVelScale=<1 1 1>
    physicsBodyDefinitions [{
      sphereCollisionRadius=0
      boundBone {boneName=dyn_flower_05}
      boxExtents=<27 22 14>
      constraintSetup {angularTargetAxis=AxisY, twistAxis=AxisZ, angularLimitsMax=<25 25 2>, angularLimitsMin=<-25 -25 -2>, angularTarget=<0 1 0>}
      localJointOffset=<0 0 -2.2>
     }]
   }
   n6 = Anim Dynamics - Bone: dyn_flower_06(ComponentPose=@n5, Alpha=0.15, ExternalForce="0,0,0") {
    angularDampingOverride=0.995
    angularSpringConstant=20000
    bAngularSpring=true
    bGravityOverrideInSimSpace=true
    bOverrideAngularDamping=true
    numSolverIterationsPostUpdate=4
    numSolverIterationsPreUpdate=16
    alphaBoolBlend {blendOption=Linear}
    boundBone {boneName=dyn_flower_06}
    componentLinearAccScale=<1 1 1>
    componentLinearVelScale=<1 1 1>
    physicsBodyDefinitions [{
      sphereCollisionRadius=0
      boundBone {boneName=dyn_flower_06}
      boxExtents=<21 19 16>
      constraintSetup {angularTargetAxis=AxisY, twistAxis=AxisZ, angularLimitsMax=<25 25 2>, angularLimitsMin=<-25 -25 -2>, angularTarget=<0 1 0>}
      localJointOffset=<0 0 -7>
     }]
   }
   n7 = Anim Dynamics - Bone: dyn_flower_07(ComponentPose=@n6, Alpha=0.23, ExternalForce="0,0,0") {
    angularDampingOverride=0.995
    angularSpringConstant=20000
    bAngularSpring=true
    bGravityOverrideInSimSpace=true
    bOverrideAngularDamping=true
    numSolverIterationsPostUpdate=4
    numSolverIterationsPreUpdate=16
    alphaBoolBlend {blendOption=Linear}
    boundBone {boneName=dyn_flower_07}
    componentLinearAccScale=<1 1 1>
    componentLinearVelScale=<1 1 1>
    physicsBodyDefinitions [{
      sphereCollisionRadius=0
      boundBone {boneName=dyn_flower_07}
      boxExtents=<30 30 15>
      constraintSetup {angularTargetAxis=AxisY, twistAxis=AxisZ, angularLimitsMax=<25 25 2>, angularLimitsMin=<-25 -25 -2>, angularTarget=<0 1 0>}
      localJointOffset=<0 0 -6>
     }]
   }
   Output Pose(Result=Component To Local(ComponentPose=@n7)) {layerGroup=None}
  ]}
 EventGraph {nodes [Event BlueprintUpdateAnimation()]}
]
variables []
